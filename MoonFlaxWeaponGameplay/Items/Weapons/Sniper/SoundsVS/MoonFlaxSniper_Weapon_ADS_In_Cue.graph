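# Sound cue for the MoonFlax sniper's aim-down-sights-in foley.
# A Random node with a single weighted entry plays one wave, Sniper_MoonFlax_ADS_In_01, so there is no variation yet.
# It uses the shared weapon reload attenuation and WeaponFoley sound class at roughly half volume.
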SoundCue /MoonFlaxWeaponGameplay/Items/Weapons/Sniper/SoundsVS/MoonFlaxSniper_Weapon_ADS_In_Cue
root=Random_1
nodes [
 {id=Random_1, children [WavePlayer_0], properties {weights [1]}}
 {id=WavePlayer_0, properties {soundWaveAssetPtr=/MoonFlaxWeaponGameplay/Items/Weapons/Sniper/SoundsVS/Waves/Sniper_MoonFlax_ADS_In_01}}
]
properties {
 attenuationSettings=&Fort_Weapon_Reload_Att
 maxDistance=2250
 soundClassObject=&Fort_WeaponFoley
 volumeMultiplier=0.501
}
