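# Sound cue for picking up the MoonFlax sniper weapon.
# A single wave player feeds the pickup wave through a modulator node left at default settings.
# It uses shared Fortnite item attenuation and the weapon foley sound class, with reduced volume.
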SoundCue /MoonFlaxWeaponGameplay/Items/Weapons/Sniper/SoundsVS/MoonFlaxSniper_Weapon_Pickup_Cue
root=Modulator_0
nodes [
 {id=Modulator_0, children [WavePlayer_0]}
 {id=WavePlayer_0, properties {soundWaveAssetPtr=/MoonFlaxWeaponGameplay/Items/Weapons/Sniper/SoundsVS/Waves/Sniper_MoonFlax_Pickup_01}}
]
properties {
 attenuationSettings=/FortMetasound/Items/AS_Fort_Item
 maxDistance=3500
 soundClassObject=&Fort_WeaponFoley
 volumeMultiplier=0.25
}
