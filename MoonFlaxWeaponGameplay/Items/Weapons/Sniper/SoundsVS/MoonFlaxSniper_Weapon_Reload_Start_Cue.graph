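# Sound cue for the start of the MoonFlax sniper weapon's reload.
# A Random node picks one of six equally weighted wave variations of the reload-start sound.
# It plays at about half volume in the weapon foley sound class, spatialized through the shared weapon reload attenuation.
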SoundCue /MoonFlaxWeaponGameplay/Items/Weapons/Sniper/SoundsVS/MoonFlaxSniper_Weapon_Reload_Start_Cue
root=Random_0
nodes [
 {id=Random_0, children [WavePlayer_0,WavePlayer_1,WavePlayer_2,WavePlayer_3,WavePlayer_4,WavePlayer_5], properties {weights [1,1,1,1,1,1]}}
 {id=WavePlayer_0, properties {soundWaveAssetPtr=/MoonFlaxWeaponGameplay/Items/Weapons/Sniper/SoundsVS/Waves/Sniper_MoonFlax_Reload_Start_01}}
 {id=WavePlayer_1, properties {soundWaveAssetPtr=/MoonFlaxWeaponGameplay/Items/Weapons/Sniper/SoundsVS/Waves/Sniper_MoonFlax_Reload_Start_02}}
 {id=WavePlayer_2, properties {soundWaveAssetPtr=/MoonFlaxWeaponGameplay/Items/Weapons/Sniper/SoundsVS/Waves/Sniper_MoonFlax_Reload_Start_03}}
 {id=WavePlayer_3, properties {soundWaveAssetPtr=/MoonFlaxWeaponGameplay/Items/Weapons/Sniper/SoundsVS/Waves/Sniper_MoonFlax_Reload_Start_04}}
 {id=WavePlayer_4, properties {soundWaveAssetPtr=/MoonFlaxWeaponGameplay/Items/Weapons/Sniper/SoundsVS/Waves/Sniper_MoonFlax_Reload_Start_05}}
 {id=WavePlayer_5, properties {soundWaveAssetPtr=/MoonFlaxWeaponGameplay/Items/Weapons/Sniper/SoundsVS/Waves/Sniper_MoonFlax_Reload_Start_06}}
]
properties {
 attenuationSettings=&Fort_Weapon_Reload_Att
 maxDistance=2250
 soundClassObject=&Fort_WeaponFoley
 volumeMultiplier=0.501
}
